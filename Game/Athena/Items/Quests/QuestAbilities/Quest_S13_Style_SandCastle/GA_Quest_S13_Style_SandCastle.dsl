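; Gameplay ability for the Season 13 SandCastle style quests that swaps the player's skin once all listed quests are completed.
; On activation off the server, it waits for quest data, counts completed quests and listens for further completions if any remain.
; Completion events increment the count until it reaches the list length.
; The swap then grants an item, forces a pawn net update and applies tag and FX gameplay effects.
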(event Ability|EventActivateAbility
  ;[1 1 1] Setup
  (|CallParentFunction)
  (if (not (Networking|IsServer))
    (bind (_owningquestmanager _bfoundallquestsonprofile) (Quests|CurrentQuestsReadyAsync (Variables|GAVars|Data|GetOwningPlayerQuestManager) (Variables|Default|GetQuests))
      (:OnCurrentQuestsReady
        (if _bfoundallquestsonprofile
          (for _array_element (Variables|Default|GetQuests)
            (bind (_playercontrollerout _completedquest) (Quest|HasPlayerCompletedQuest (Fortnite|GetFortPlayerControllerfromActor (Variables|GAVars|Data|GetOwningPlayerPawn)) _array_element))
            (if _completedquest
              (Math|Integer|IncrementInt (Variables|Default|GetQuestsCompletedCount))))
          (if (< (Variables|Default|GetQuestsCompletedCount) (Utilities|Array|Length (Variables|Default|GetQuests)))
            (Quests|BindEventtoOnQuestsCompleted (Variables|GAVars|Data|GetOwningPlayerQuestManager) (delegate Custom|QuestsCompleted))))))))

(event Ability|EventFailedToActivatePassiveAbility
  (params (ActorInfo :type Gameplay%20Ability%20Actor%20Info%20Structure))
  (|CallParentFunction))

(event Custom|QuestsCompleted
  (params (Quests :type Array%20of%20Fort%20Quest%20Item%20Object%20References))
  ;[1 1 1] Quest completed
  (for _array_element Quests
    (if (Utilities|Array|ContainsItem (Variables|Default|GetQuests) (Quest|GetQuestDefinition _array_element))
      (Math|Integer|IncrementInt (Variables|Default|GetQuestsCompletedCount))))
  ;[1 1 1] Skin swap
  (if (>= (Variables|Default|GetQuestsCompletedCount) (Utilities|Array|Length (Variables|Default|GetQuests)))
    (CallFunction|SkinSwap)
    (Quests|UnbindallEventsfromOnQuestsCompleted (Variables|GAVars|Data|GetOwningPlayerQuestManager))))

(event Custom|SkinSwap
  ;[1 1 1] Skin swap
  (Inventory|GiveItemtoInventoryOwner (Fortnite|GetFortPlayerControllerfromActor (Variables|GAVars|Data|GetOwningPlayerPawn)) (Variables|Default|GetAGID))
  (Networking|ForceNetUpdate (Variables|GAVars|Data|GetOwningPlayerPawn))
  (Ability|ApplyGameplayEffectToOwner (Variables|Default|GetSwapTagGE))
  (GameplayEffects|ApplyGameplayEffectToSelf :self (Ability|GetAbilitySystemComponent (Variables|GAVars|Data|GetOwningPlayerPawn)) :GameplayEffectClass (Variables|Default|GetSwapFXGE) :Level 4.0))
